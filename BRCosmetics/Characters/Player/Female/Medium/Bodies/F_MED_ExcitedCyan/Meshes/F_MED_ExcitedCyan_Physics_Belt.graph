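# Physics asset for the belt of the F_MED_ExcitedCyan female character body, covering dangling fish and mouse accessories.
# Both accessories are capsule-body chains from the pelvis, with free angular joints at their bases and tight swing and twist limits below.
# The mouse chain branches into two ears and a two-segment tail.
# Root, pelvis, right thigh and right calf capsules add body volumes, linked by free angular constraints.
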PhysicsAsset /BRCosmetics/Characters/Player/Female/Medium/Bodies/F_MED_ExcitedCyan/Meshes/F_MED_ExcitedCyan_Physics_Belt
bodies {
 calf_r [{
   shapeType=Capsule
   length=49.5872
   radius=7.9678
   shapeName=calf_l_capsule
   center=<23.7705 2.0394 -0.4514>
   rotation=<rot 87.4286 -177.4242 -0.0002>
  }]
 dyn_fish_01 [{shapeType=Capsule, length=1.7517, radius=1.3459, shapeName=dyn_fish_01_capsule, center=<2.1947 0 0>, rotation=<rot 90 0 0>}]
 dyn_fish_02 [{shapeType=Capsule, length=1.858, radius=1.0722, shapeName=dyn_fish_02_capsule, center=<1.6319 0 0>, rotation=<rot 90 0 0>}]
 dyn_fish_03 [{shapeType=Capsule, length=9.6967, radius=3.9312, shapeName=dyn_fish_03_capsule, center=<4.3705 0.2597 -1.9856>, rotation=<rot 0 0 -90>}]
 dyn_fish_04 [{
   shapeType=Capsule
   length=8.4103
   radius=4.1925
   shapeName=dyn_fish_04_capsule
   center=<8.7887 0 0.6261>
   rotation=<rot 85.9249 179.9999 179.9998>
  }]
 dyn_fish_base [{shapeType=Capsule, length=0, radius=0, shapeName=dyn_fish_base_capsule}]
 dyn_mouse_ear_l_01 [{
   shapeType=Capsule
   length=1.0723
   radius=1.4138
   shapeName=dyn_mouse_ear_l_01_capsule
   center=<1.0619 0 0.1515>
   rotation=<rot 83.9966 -0.0008 -0.001>
  }]
 dyn_mouse_ear_r_01 [{
   shapeType=Capsule
   length=1.446
   radius=1.3076
   shapeName=dyn_mouse_ear_r_01_capsule
   center=<1.2398 0.0805 -0.1352>
   rotation=<rot 0 -83.0109 -78.9678>
  }]
 dyn_mouse_r_01 [{
   shapeType=Capsule
   length=4.7775
   radius=1.2214
   shapeName=dyn_mouse_r_01_capsule
   center=<2.7291 -0.3414 0.1536>
   rotation=<rot 0 -84.6011 -88.879>
  }]
 dyn_mouse_r_02 [{
   shapeType=Capsule
   length=5.6639
   radius=2.8242
   shapeName=dyn_mouse_r_02_capsule
   center=<2.3348 0.7641 0.2181>
   rotation=<rot 0 177.5653 -88.3903>
  }]
 dyn_mouse_r_base [{shapeType=Capsule, length=0.505, radius=0.505, shapeName=dyn_mouse_r_base_capsule, rotation=<rot 0 -90 0>}]
 dyn_mouse_tail_r_01 [{
   shapeType=Capsule
   length=2.4645
   radius=1.4251
   shapeName=dyn_mouse_tail_r_01_capsule
   center=<1.7201 -0.0439 0.0328>
   rotation=<rot 0 -90.8181 -87.0689>
  }]
 dyn_mouse_tail_r_02 [{
   shapeType=Capsule
   length=3.3589
   radius=1.2011
   shapeName=dyn_mouse_tail_r_02_capsule
   center=<1.5818 1.2858 -0.1274>
   rotation=<rot 0.5874 -60.6734 -87.8627>
  }]
 pelvis [{shapeType=Capsule, length=6.9136, radius=10.5243, shapeName=pelvis_capsule, center=<-5.9542 -2.1131 0>}]
 root [{shapeType=Capsule, length=0, radius=82.1141, shapeName=root_capsule, center=<0 -3.371 -78.6922>, rotation=<rot 0 -90 0>}]
 thigh_r [{
   shapeType=Capsule
   length=37.9546
   radius=7.8679
   shapeName=thigh_l_capsule
   center=<22.9605 -0.1756 0.7822>
   rotation=<rot 84.0337 178.0328 2.727>
  }]
}
constraints [
 {
  bone1Name=dyn_fish_base
  bone2Name=pelvis
  swing1LimitDegrees=45
  swing1Motion=Free
  swing2LimitDegrees=45
  swing2Motion=Free
  twistLimitDegrees=45
  twistMotion=Free
 }
 {bone1Name=dyn_fish_01, bone2Name=dyn_fish_base, swing1LimitDegrees=10, swing2LimitDegrees=20, twistLimitDegrees=10}
 {bone1Name=dyn_fish_02, bone2Name=dyn_fish_01, swing1LimitDegrees=30, swing2LimitDegrees=30, twistLimitDegrees=10}
 {bone1Name=dyn_fish_03, bone2Name=dyn_fish_02, swing1LimitDegrees=10, swing2LimitDegrees=20, twistLimitDegrees=10}
 {bone1Name=dyn_fish_04, bone2Name=dyn_fish_03, swing1LimitDegrees=40, swing2LimitDegrees=30, twistLimitDegrees=10}
 {
  bone1Name=thigh_r
  bone2Name=pelvis
  swing1LimitDegrees=45
  swing1Motion=Free
  swing2LimitDegrees=45
  swing2Motion=Free
  twistLimitDegrees=45
  twistMotion=Free
 }
 {
  bone1Name=calf_r
  bone2Name=thigh_r
  swing1LimitDegrees=45
  swing1Motion=Free
  swing2LimitDegrees=45
  swing2Motion=Free
  twistLimitDegrees=45
  twistMotion=Free
 }
 {
  bone1Name=pelvis
  bone2Name=root
  swing1LimitDegrees=45
  swing1Motion=Free
  swing2LimitDegrees=45
  swing2Motion=Free
  twistLimitDegrees=45
  twistMotion=Free
 }
 {
  bone1Name=dyn_mouse_r_base
  bone2Name=pelvis
  swing1LimitDegrees=45
  swing1Motion=Free
  swing2LimitDegrees=45
  swing2Motion=Free
  twistLimitDegrees=45
  twistMotion=Free
 }
 {bone1Name=dyn_mouse_r_01, bone2Name=dyn_mouse_r_base, swing1LimitDegrees=30, swing2LimitDegrees=20, twistLimitDegrees=20}
 {bone1Name=dyn_mouse_r_02, bone2Name=dyn_mouse_r_01, swing1LimitDegrees=10, swing2LimitDegrees=0, twistLimitDegrees=20}
 {bone1Name=dyn_mouse_ear_l_01, bone2Name=dyn_mouse_r_02, swing1LimitDegrees=10, swing2LimitDegrees=10, twistLimitDegrees=5}
 {bone1Name=dyn_mouse_tail_r_01, bone2Name=dyn_mouse_r_02, swing1LimitDegrees=10, swing2LimitDegrees=0, twistLimitDegrees=0}
 {bone1Name=dyn_mouse_tail_r_02, bone2Name=dyn_mouse_tail_r_01, swing1LimitDegrees=20, swing2LimitDegrees=0, twistLimitDegrees=0}
 {bone1Name=dyn_mouse_ear_r_01, bone2Name=dyn_mouse_r_02, swing1LimitDegrees=5, swing2LimitDegrees=5, twistLimitDegrees=0}
]
