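# Physics asset for the hip chain and bag on the F_MED_ExcitedCyan female character body.
# Back, front and left-support chains and a bag hang from pelvis through base bodies, with limited swing and twist down each chain, the bag nearly rigid.
# Angular-locked constraints join both chain tips to the last support body, closing a loop.
# Capsules on root, pelvis, left thigh and calf supply body shapes, linked by free-rotation constraints.
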PhysicsAsset /BRCosmetics/Characters/Player/Female/Medium/Bodies/F_MED_ExcitedCyan/Meshes/F_MED_ExcitedCyan_Physics_Chain
bodies {
 calf_l [{
   shapeType=Capsule
   length=48.3537
   radius=8.1616
   shapeName=calf_l_capsule
   center=<-26.6757 -2.6651 0>
   rotation=<rot 90 -179.8881 178.5002>
  }]
 chain_support_l_01 [{shapeType=Capsule, length=11.242, radius=2, shapeName=chain_support_l_01_capsule, center=<7.6317 0 0>, rotation=<rot 0 -90 90>}]
 chain_support_l_02 [{
   shapeType=Capsule
   length=11.1601
   radius=4.0395
   shapeName=chain_support_l_02_capsule
   center=<1.4951 -0.6659 2.4515>
   rotation=<rot -0.0001 0.0002 83.9642>
  }]
 chain_support_l_base [{shapeType=Capsule, length=0, radius=0, shapeName=chain_support_l_base_capsule}]
 dyn_bag_01 [{shapeType=Capsule, length=14.272, radius=5.819, shapeName=dyn_bag_01_capsule, center=<0 0.5969 -2.7039>, rotation=<rot 90 0 0>}]
 dyn_bag_base [{shapeType=Capsule, length=0.505, radius=0.505, shapeName=dyn_bag_base_capsule, rotation=<rot 0 -90 0>}]
 dyn_chain_bk_01 [{shapeType=Capsule, length=3.2799, radius=1.8622, shapeName=dyn_chain_bk_01_capsule, center=<2.6463 -0.6161 0>, rotation=<rot 90 0 0>}]
 dyn_chain_bk_02 [{
   shapeType=Capsule
   length=4.0518
   radius=1.8933
   shapeName=dyn_chain_bk_02_capsule
   center=<2.5721 -0.8745 0>
   rotation=<rot 90 -78.9374 -74.7542>
  }]
 dyn_chain_bk_03 [{
   shapeType=Capsule
   length=7.1612
   radius=2.4727
   shapeName=dyn_chain_bk_03_capsule
   center=<4.5413 -0.9272 0>
   rotation=<rot 90 180 175.5899>
  }]
 dyn_chain_bk_04 [{
   shapeType=Capsule
   length=7.1041
   radius=3.6301
   shapeName=dyn_chain_bk_04_capsule
   center=<4.2106 -1.7276 -0.0417>
   rotation=<rot 85.9033 -179.9936 179.9224>
  }]
 dyn_chain_bk_05 [{
   shapeType=Capsule
   length=12.74
   radius=4.0565
   shapeName=dyn_chain_bk_05_capsule
   center=<4.534 -1.5652 0>
   rotation=<rot 90 -161.834 -165.0871>
  }]
 dyn_chain_bk_base [{shapeType=Capsule, length=0, radius=0, shapeName=dyn_chain_bk_base_capsule}]
 dyn_chain_frnt_01 [{shapeType=Capsule, length=3.0029, radius=1.3899, shapeName=dyn_chain_frnt_01_capsule, center=<1.7007 0.3653 0>, rotation=<rot 90 0 0>}]
 dyn_chain_frnt_02 [{
   shapeType=Capsule
   length=3.1491
   radius=2.5164
   shapeName=dyn_chain_frnt_02_capsule
   center=<2.5179 0.7276 0.3027>
   rotation=<rot 90 -45.1254 -41.7546>
  }]
 dyn_chain_frnt_03 [{
   shapeType=Capsule
   length=4.7605
   radius=3.0045
   shapeName=dyn_chain_frnt_03_capsule
   center=<5.3003 -0.1667 1.24>
   rotation=<rot 78.24 156.2572 -179.9999>
  }]
 dyn_chain_frnt_04 [{
   shapeType=Capsule
   length=6.4745
   radius=3.1909
   shapeName=dyn_chain_frnt_04_capsule
   center=<4.3461 0.7354 1.098>
   rotation=<rot 75.0567 168.5645 -171.2253>
  }]
 dyn_chain_frnt_base [{shapeType=Capsule, length=0, radius=0, shapeName=dyn_chain_frnt_base_capsule}]
 pelvis [{shapeType=Capsule, length=6.9136, radius=10.5243, shapeName=pelvis_capsule, center=<-5.9542 -2.1131 0>}]
 root [{shapeType=Capsule, length=0, radius=73.6701, shapeName=root_capsule, center=<0 -14.1471 -71.7988>, rotation=<rot 0 -90 0>}]
 thigh_l [{
   shapeType=Capsule
   length=36.9173
   radius=8.9052
   shapeName=thigh_l_capsule
   center=<-23.0524 0.1576 1.4932>
   rotation=<rot 89.263 -162.063 -157.3827>
  }]
}
constraints [
 {
  bone1Name=thigh_l
  bone2Name=pelvis
  swing1LimitDegrees=45
  swing1Motion=Free
  swing2LimitDegrees=45
  swing2Motion=Free
  twistLimitDegrees=45
  twistMotion=Free
 }
 {
  bone1Name=dyn_chain_bk_base
  bone2Name=pelvis
  swing1LimitDegrees=45
  swing1Motion=Free
  swing2LimitDegrees=45
  swing2Motion=Free
  twistLimitDegrees=45
  twistMotion=Free
 }
 {bone1Name=dyn_chain_bk_01, bone2Name=dyn_chain_bk_base, swing1LimitDegrees=20, swing2LimitDegrees=20, twistLimitDegrees=10}
 {bone1Name=dyn_chain_bk_02, bone2Name=dyn_chain_bk_01, swing1LimitDegrees=20, swing2LimitDegrees=20, twistLimitDegrees=10}
 {bone1Name=dyn_chain_bk_03, bone2Name=dyn_chain_bk_02, swing1LimitDegrees=20, swing2LimitDegrees=20, twistLimitDegrees=10}
 {bone1Name=dyn_chain_bk_04, bone2Name=dyn_chain_bk_03, swing1LimitDegrees=30, swing2LimitDegrees=30, twistLimitDegrees=10}
 {bone1Name=dyn_chain_bk_05, bone2Name=dyn_chain_bk_04, swing1LimitDegrees=30, swing2LimitDegrees=30, twistLimitDegrees=10}
 {
  bone1Name=dyn_chain_frnt_base
  bone2Name=pelvis
  swing1LimitDegrees=45
  swing1Motion=Free
  swing2LimitDegrees=45
  swing2Motion=Free
  twistLimitDegrees=45
  twistMotion=Free
 }
 {bone1Name=dyn_chain_frnt_01, bone2Name=dyn_chain_frnt_base, swing1LimitDegrees=20, swing2LimitDegrees=20, twistLimitDegrees=10}
 {bone1Name=dyn_chain_frnt_02, bone2Name=dyn_chain_frnt_01, swing1LimitDegrees=30, swing2LimitDegrees=30, twistLimitDegrees=10}
 {bone1Name=dyn_chain_frnt_03, bone2Name=dyn_chain_frnt_02, swing1LimitDegrees=30, swing2LimitDegrees=30, twistLimitDegrees=0}
 {bone1Name=dyn_chain_frnt_04, bone2Name=dyn_chain_frnt_03, swing1LimitDegrees=30, swing2LimitDegrees=30, twistLimitDegrees=10}
 {
  bone1Name=chain_support_l_base
  bone2Name=pelvis
  swing1LimitDegrees=45
  swing1Motion=Free
  swing2LimitDegrees=45
  swing2Motion=Free
  twistLimitDegrees=45
  twistMotion=Free
 }
 {bone1Name=chain_support_l_01, bone2Name=chain_support_l_base, swing1LimitDegrees=20, swing2LimitDegrees=65, twistLimitDegrees=20}
 {bone1Name=chain_support_l_02, bone2Name=chain_support_l_01, swing1LimitDegrees=45, swing2LimitDegrees=45, twistLimitDegrees=45}
 {
  bone1Name=dyn_bag_base
  bone2Name=pelvis
  swing1LimitDegrees=45
  swing1Motion=Free
  swing2LimitDegrees=45
  swing2Motion=Free
  twistLimitDegrees=45
  twistMotion=Free
 }
 {bone1Name=dyn_bag_01, bone2Name=dyn_bag_base, swing1LimitDegrees=2, swing2LimitDegrees=2, twistLimitDegrees=0}
 {
  bone1Name=calf_l
  bone2Name=thigh_l
  swing1LimitDegrees=45
  swing1Motion=Free
  swing2LimitDegrees=45
  swing2Motion=Free
  twistLimitDegrees=45
  twistMotion=Free
 }
 {
  bone1Name=pelvis
  bone2Name=root
  swing1LimitDegrees=45
  swing1Motion=Free
  swing2LimitDegrees=45
  swing2Motion=Free
  twistLimitDegrees=45
  twistMotion=Free
 }
 {
  bone1Name=chain_support_l_02
  bone2Name=dyn_chain_bk_05
  swing1LimitDegrees=45
  swing1Motion=Locked
  swing2LimitDegrees=45
  swing2Motion=Locked
  twistLimitDegrees=45
  twistMotion=Locked
 }
 {
  bone1Name=dyn_chain_frnt_04
  bone2Name=chain_support_l_02
  swing1LimitDegrees=45
  swing1Motion=Locked
  swing2LimitDegrees=45
  swing2Motion=Locked
  twistLimitDegrees=45
  twistMotion=Locked
 }
]
