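# Control Rig for the F_MED_ExcitedCyan face accessory mesh that offsets hair bones, including the left and right big braid bases, as the head turns.
# A spherical pose reader measures the head bone's direction relative to spine_05 within an offset cone with falloff, with debug drawing left on.
# Its output is smoothed by a clamped alpha interpolation and drives the weight of a single Modify Transforms node.
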ControlRigBlueprint /BRCosmetics/Characters/Player/Female/Medium/Bodies/F_MED_ExcitedCyan/Meshes/Parts/F_MED_ExcitedCyan_FaceAcc_CtrlRig
graphs [
 RigVMFunctionLibrary {links [], nodes []}
 RigVMModel {nodes [
   n1 = &RigUnit_SphericalPoseReader(
    ActiveRegionScaleFactors=(PositiveWidth=1,NegativeWidth=1,PositiveHeight=1,NegativeHeight=1)
    ActiveRegionSize=0.01
    Debug=(bDrawDebug=True,bDraw2D=False,bDrawLocalAxes=False,DebugScale=25,DebugSegments=20,DebugThickness=0.25)
    DriverAxis=(X=1,Y=0,Z=0)
    DriverItem=(Type=Bone,Name="head")
    ExecutePin=()
    FalloffRegionScaleFactors=(PositiveWidth=1,NegativeWidth=1,PositiveHeight=1,NegativeHeight=1)
    FalloffSize=0.15
    FlipHeightScaling=False
    FlipWidthScaling=False
    OptionalParentItem=(Type=Bone,Name="spine_05")
    RotationOffset=(X=-65,Y=0,Z=60)
   )
   n2 = &RigVMFunction_AlphaInterp(
    InterpSpeedDecreasing=24
    InterpSpeedIncreasing=24
    OutRange=(Max=1)
    Scale=1
    bClampResult=true
    bMapRange=true
    Bias=~
    ClampMin=~
    Value=@n1.OutputParam
   )
   > &RigUnit_ModifyTransforms(
    ItemToModify="((Item=(Type=Bone,Name=\"dyn_big_braid_r_base\"),Transform=(Rotation=(X=0,Y=0,Z=0.2164,W=0.9763),Translation=(X=0,Y=2,Z=0),Scale3D=(X=1,Y=1,Z=1))),(Item=(Type=Bone,Name=\"dyn_big_braid_l_base\"),Transform=(Rotation=(X=0,Y=0,Z=0,W=1),Translation=(X=0,Y=-2,Z=0), ...(truncated)"
    WeightMinimum=0
    Weight=@n2.Result
   )
   RigUnit_BeginExecution = &RigUnit_BeginExecution()
  ]}
]
hierarchy [
 root {loc=<0 0 0>, rot=<0 0 0 1>}
 pelvis {parent=root, loc=<0 -0.5301 98.0571>, rot=<0.0127 0.707 -0.0127 -0.707>}
 spine_01 {parent=pelvis, loc=<6.1208 -0.9328 0>, rot=<0 0 0.034 -0.9994>}
 spine_02 {parent=spine_01, loc=<6.5996 -0.2095 0>, rot=<0 0 0.0238 -0.9997>}
 spine_03 {parent=spine_02, loc=<6.4914 -0.1237 0>, rot=<0 0 0.0149 -0.9999>}
 spine_04 {parent=spine_03, loc=<6.9598 0.0794 0>, rot=<0 0 -0.0132 -0.9999>}
 spine_05 {parent=spine_04, loc=<6.9987 0 0>, rot=<0 0 -0.0015 -1>}
 neck_01 {parent=spine_05, loc=<11.3138 0.1529 0>, rot=<0 0 -0.2003 -0.9797>}
 neck_02 {parent=neck_01, loc=<4.5627 0 0>, rot=<0 0 0.0504 -0.9987>}
 head {parent=neck_02, loc=<5.3914 0 0>, rot=<0 0 0.115 -0.9934>}
 dyn_big_braid_l_base {parent=head, loc=<-3.909 -3.9065 -5.8322>, rot=<-0.2426 -0.9688 -0.0127 -0.0488>}
 dyn_big_braid_l_01 {parent=dyn_big_braid_l_base, loc=<0 0 0>, rot=<0 0 0 1>}
 dyn_big_braid_l_03 {parent=dyn_big_braid_l_01, loc=<5.6935 0.2777 -0.1934>, rot=<-0.0005 -0.0348 -0.1725 -0.9844>}
 dyn_big_braid_l_05 {parent=dyn_big_braid_l_03, loc=<6 0 0>, rot=<0.0218 0.0242 0.1112 -0.9933>}
 dyn_big_braid_l_07 {parent=dyn_big_braid_l_05, loc=<6.8971 -0.2169 0>, rot=<0 0 0.0278 -0.9996>}
 dyn_big_braid_tail_l_base {parent=dyn_big_braid_l_07, loc=<-0.3417 0.3078 0.3927>, rot=<0.2508 -0.1941 0.8518 0.417>}
 dyn_quiff_l_base {parent=head, loc=<9.9016 8.3536 -3.802>, rot=<-0.1045 -0.8911 -0.236 -0.3732>}
 dyn_quiff_l_01 {parent=dyn_quiff_l_base, loc=<0 0 0>, rot=<0 0 0 1>}
 dyn_small_braid_l_base {parent=head, loc=<7.6542 6.0126 -5.9771>, rot=<0.0199 -0.9998 0 0>}
 dyn_small_braid_l_01 {parent=dyn_small_braid_l_base, loc=<0 0 0>, rot=<0 0 0 1>}
 dyn_small_braid_l_02 {parent=dyn_small_braid_l_01, loc=<4.8454 0.0377 0.1528>, rot=<0.0007 -0.0209 -0.0385 -0.999>}
 dyn_small_braid_l_03 {parent=dyn_small_braid_l_02, loc=<4.4684 -0.0237 0.013>, rot=<0.0006 0.0195 0.0327 -0.9993>}
 dyn_hair_bk_l_01 {parent=head, loc=<3.4327 -6.3881 -6.1558>, rot=<-0.0322 -0.8697 0.4887 0.0617>}
 dyn_big_braid_r_base {parent=head, loc=<-3.8797 -3.9062 6.1176>, rot=<0.9688 -0.2426 -0.0488 0.0127>}
 dyn_big_braid_r_01 {parent=dyn_big_braid_r_base, loc=<0 0 0>, rot=<0 0 0 1>}
 dyn_big_braid_r_03 {parent=dyn_big_braid_r_01, loc=<-5.6939 -0.2775 0.289>, rot=<-0.0005 -0.0348 -0.1725 -0.9844>}
 dyn_big_braid_r_05 {parent=dyn_big_braid_r_03, loc=<-5.9998 -0.0002 0>, rot=<0.0218 0.0242 0.1112 -0.9933>}
 dyn_big_braid_r_07 {parent=dyn_big_braid_r_05, loc=<-6.8971 0.217 0>, rot=<0 0 0.0278 -0.9996>}
 dyn_big_braid_tail_r_base {parent=dyn_big_braid_r_07, loc=<0.3417 -0.3078 -0.3927>, rot=<0.2508 -0.1941 0.8518 0.417>}
 dyn_quiff_r_base {parent=head, loc=<9.9021 8.3536 3.802>, rot=<0.8911 -0.1045 -0.3732 0.236>}
 dyn_quiff_r_01 {parent=dyn_quiff_r_base, loc=<0 0 0>, rot=<0 0 0 1>}
 dyn_small_braid_r_base {parent=head, loc=<7.6543 6.0126 5.9771>, rot=<0.9998 0.0199 0 0>}
 dyn_small_braid_r_01 {parent=dyn_small_braid_r_base, loc=<0 0 0>, rot=<0 0 0 1>}
 dyn_small_braid_r_02 {parent=dyn_small_braid_r_01, loc=<-4.8452 -0.0377 -0.1528>, rot=<0.0007 -0.0209 -0.0385 -0.999>}
 dyn_small_braid_r_03 {parent=dyn_small_braid_r_02, loc=<-4.4689 0.0237 -0.0131>, rot=<0.0006 0.0195 0.0327 -0.9993>}
 dyn_hair_bk_r_01 {parent=head, loc=<5.8512 -6.4885 5.7353>, rot=<0.9062 0.0513 -0.0719 -0.4135>}
 clavicle_l {parent=spine_05, loc=<5.4831 -0.1053 -3.2946>, rot=<0.0465 -0.7289 -0.009 -0.6829>}
 upperarm_l {parent=clavicle_l, loc=<9.8658 -1.2137 -0.4105>, rot=<0.0164 -0.3245 -0.0269 -0.9454>}
 clavicle_r {parent=spine_05, loc=<5.4831 -0.1053 3.2946>, rot=<0.7289 0.0465 -0.6829 0.009>}
 upperarm_r {parent=clavicle_r, loc=<-9.8658 1.2137 0.4105>, rot=<0.0164 -0.3245 -0.0269 -0.9454>}
]
