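# Physics asset for the F_MED_ExcitedCyan female cosmetic face-accessory mesh, setting up secondary motion for its paired big braids, small braids and quiffs.
# Capsule bodies cover the upper spine, neck, head, clavicles, upper arms and each hair chain, with big braids using only alternate numbered bones.
# Hair base bodies, tiny or zero-sized, join to the head with free angular constraints, as do the torso joints.
# Downstream hair segments carry tighter, per-joint swing and twist limits.
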PhysicsAsset /BRCosmetics/Characters/Player/Female/Medium/Bodies/F_MED_ExcitedCyan/Meshes/Parts/F_MED_ExcitedCyan_FaceAcc_Physics
bodies {
 clavicle_l [{
   shapeType=Capsule
   length=10.7406
   radius=4.2244
   shapeName=clavicle_l_capsule
   center=<5.363 -1.2644 -0.6876>
   rotation=<rot 84.5416 -7.3888 -0.0015>
  }]
 clavicle_r [{
   shapeType=Capsule
   length=6.5156
   radius=4
   shapeName=clavicle_r_capsule
   center=<-3.8185 1.2852 0.6005>
   rotation=<rot 85.6284 -4.4798 -0.002>
  }]
 dyn_big_braid_l_01 [{shapeType=Capsule, length=6.6759, radius=2.5451, shapeName=dyn_big_braid_l_01_capsule, center=<4.5842 0 0.1714>, rotation=<rot 90 0 0>}]
 dyn_big_braid_l_03 [{shapeType=Capsule, length=3.6514, radius=2.4401, shapeName=dyn_big_braid_l_03_capsule, center=<2.6368 0 0.1961>, rotation=<rot 90 0 0>}]
 dyn_big_braid_l_05 [{shapeType=Capsule, length=4.831, radius=2.051, shapeName=dyn_big_braid_l_05_capsule, center=<4.0484 0.7878 0>, rotation=<rot 90 0 0>}]
 dyn_big_braid_l_07 [{shapeType=Capsule, length=2.5595, radius=2.1585, shapeName=dyn_big_braid_l_07_capsule, center=<2.5404 0.1237 0>, rotation=<rot 90 0 0>}]
 dyn_big_braid_l_base [{shapeType=Capsule, length=0, radius=0, shapeName=dyn_big_braid_l_base_capsule, rotation=<rot 90 0 0>}]
 dyn_big_braid_r_01 [{shapeType=Capsule, length=4.4068, radius=2.1373, shapeName=dyn_big_braid_r_01_capsule, center=<-2.8588 0 0>, rotation=<rot 90 0 0>}]
 dyn_big_braid_r_03 [{shapeType=Capsule, length=4.7794, radius=1.9702, shapeName=dyn_big_braid_r_03_capsule, center=<-3.3136 0 0>, rotation=<rot 90 0 0>}]
 dyn_big_braid_r_05 [{shapeType=Capsule, length=3.7989, radius=2.1258, shapeName=dyn_big_braid_r_05_capsule, center=<-3.1168 0 0>, rotation=<rot 90 0 0>}]
 dyn_big_braid_r_07 [{shapeType=Capsule, length=2.3704, radius=2.1241, shapeName=dyn_big_braid_r_07_capsule, center=<-2.8404 0.4199 0>, rotation=<rot 90 0 0>}]
 dyn_big_braid_r_base [{shapeType=Capsule, length=0, radius=0, shapeName=dyn_big_braid_r_base_capsule, rotation=<rot 90 0 0>}]
 dyn_quiff_l_01 [{shapeType=Capsule, length=2.1823, radius=3.3177, shapeName=dyn_quiff_l_01_capsule, center=<2.6449 0 0>, rotation=<rot 0 0 90>}]
 dyn_quiff_l_base [{shapeType=Capsule, length=0.505, radius=0.505, shapeName=dyn_quiff_l_base_capsule, rotation=<rot 0 -90 0>}]
 dyn_quiff_r_01 [{shapeType=Capsule, length=1.6465, radius=4.131, shapeName=dyn_quiff_r_01_capsule, center=<-2.7855 0 0>, rotation=<rot 0 0 90>}]
 dyn_quiff_r_base [{shapeType=Capsule, length=0.505, radius=0.505, shapeName=dyn_quiff_r_base_capsule, rotation=<rot 0 -90 0>}]
 dyn_small_braid_l_01 [{shapeType=Capsule, length=4, radius=1, shapeName=dyn_small_braid_l_01_capsule, center=<2.2606 0 0>, rotation=<rot 90 0 0>}]
 dyn_small_braid_l_02 [{shapeType=Capsule, length=4, radius=1, shapeName=dyn_small_braid_l_02_capsule, center=<2.2606 0 0>, rotation=<rot 90 0 0>}]
 dyn_small_braid_l_03 [{shapeType=Capsule, length=5.6144, radius=0.8812, shapeName=dyn_small_braid_l_03_capsule, center=<3.0268 0 0>, rotation=<rot 90 0 0>}]
 dyn_small_braid_l_base [{shapeType=Capsule, length=0.505, radius=0.505, shapeName=dyn_small_braid_l_base_capsule}]
 dyn_small_braid_r_01 [{shapeType=Capsule, length=4, radius=1, shapeName=dyn_small_braid_r_01_capsule, center=<-2.064 0 0>, rotation=<rot 90 0 0>}]
 dyn_small_braid_r_02 [{shapeType=Capsule, length=4, radius=1, shapeName=dyn_small_braid_r_02_capsule, center=<-2.064 0 0>, rotation=<rot 90 0 0>}]
 dyn_small_braid_r_03 [{shapeType=Capsule, length=5.6144, radius=0.8812, shapeName=dyn_small_braid_r_03_capsule, center=<-3.3403 0 0>, rotation=<rot 90 0 0>}]
 dyn_small_braid_r_base [{shapeType=Capsule, length=0.505, radius=0.505, shapeName=dyn_small_braid_r_base_capsule}]
 head [{shapeType=Capsule, length=9.9618, radius=6.0232, shapeName=head_capsule, center=<2.5415 1.7516 0>, rotation=<rot 90 0 0>}]
 neck_01 [{shapeType=Capsule, length=13.6209, radius=3.7058, shapeName=neck_01_capsule, center=<0 1.6059 0>, rotation=<rot 90 0 0>}]
 spine_04 [{shapeType=Capsule, length=7.2837, radius=10.2968, shapeName=spine_04_capsule, center=<0 0.5883 0>, rotation=<rot 0 -90 0>}]
 spine_05 [
  {shapeType=Capsule, length=37.1931, radius=15, shapeName=spine_05_capsule, center=<11.5956 -22.298 0>}
  {shapeType=Capsule, length=16.2161, radius=5.5904, shapeName=spine_05_capsule, center=<1.3941 0.5699 0>}
 ]
 upperarm_l [{shapeType=Capsule, length=20, radius=4, shapeName=upperarm_l_capsule, center=<12.4662 -3.5778 -2.3925>, rotation=<rot 90 0 0>}]
 upperarm_r [{shapeType=Capsule, length=25.6391, radius=4, shapeName=upperarm_r_capsule, center=<-12.6256 1.5117 -0.9815>, rotation=<rot 90 0 0>}]
}
constraints [
 {
  bone1Name=clavicle_r
  bone2Name=spine_05
  swing1LimitDegrees=45
  swing1Motion=Free
  swing2LimitDegrees=45
  swing2Motion=Free
  twistLimitDegrees=45
  twistMotion=Free
 }
 {
  bone1Name=upperarm_r
  bone2Name=clavicle_r
  swing1LimitDegrees=45
  swing1Motion=Free
  swing2LimitDegrees=45
  swing2Motion=Free
  twistLimitDegrees=45
  twistMotion=Free
 }
 {
  bone1Name=neck_01
  bone2Name=spine_05
  swing1LimitDegrees=45
  swing1Motion=Free
  swing2LimitDegrees=45
  swing2Motion=Free
  twistLimitDegrees=45
  twistMotion=Free
 }
 {
  bone1Name=clavicle_l
  bone2Name=spine_05
  swing1LimitDegrees=45
  swing1Motion=Free
  swing2LimitDegrees=45
  swing2Motion=Free
  twistLimitDegrees=45
  twistMotion=Free
 }
 {
  bone1Name=upperarm_l
  bone2Name=clavicle_l
  swing1LimitDegrees=45
  swing1Motion=Free
  swing2LimitDegrees=45
  swing2Motion=Free
  twistLimitDegrees=45
  twistMotion=Free
 }
 {
  bone1Name=head
  bone2Name=neck_01
  swing1LimitDegrees=45
  swing1Motion=Free
  swing2LimitDegrees=45
  swing2Motion=Free
  twistLimitDegrees=45
  twistMotion=Free
 }
 {
  bone1Name=dyn_big_braid_l_base
  bone2Name=head
  swing1LimitDegrees=30
  swing1Motion=Free
  swing2LimitDegrees=30
  swing2Motion=Free
  twistLimitDegrees=10
  twistMotion=Free
 }
 {bone1Name=dyn_big_braid_l_01, bone2Name=dyn_big_braid_l_base, swing1LimitDegrees=30, swing2LimitDegrees=20, twistLimitDegrees=5}
 {bone1Name=dyn_big_braid_l_03, bone2Name=dyn_big_braid_l_01, swing1LimitDegrees=35, swing2LimitDegrees=5, twistLimitDegrees=5}
 {bone1Name=dyn_big_braid_l_05, bone2Name=dyn_big_braid_l_03, swing1LimitDegrees=30, swing2LimitDegrees=15, twistLimitDegrees=0}
 {bone1Name=dyn_big_braid_l_07, bone2Name=dyn_big_braid_l_05, swing1LimitDegrees=20, swing2LimitDegrees=15, twistLimitDegrees=0}
 {
  bone1Name=dyn_big_braid_r_base
  bone2Name=head
  swing1LimitDegrees=30
  swing1Motion=Free
  swing2LimitDegrees=30
  swing2Motion=Free
  twistLimitDegrees=10
  twistMotion=Free
 }
 {bone1Name=dyn_big_braid_r_07, bone2Name=dyn_big_braid_r_05, swing1LimitDegrees=30, swing2LimitDegrees=15, twistLimitDegrees=0}
 {bone1Name=dyn_big_braid_r_05, bone2Name=dyn_big_braid_r_03, swing1LimitDegrees=30, swing2LimitDegrees=15, twistLimitDegrees=0}
 {bone1Name=dyn_big_braid_r_03, bone2Name=dyn_big_braid_r_01, swing1LimitDegrees=30, swing2LimitDegrees=5, twistLimitDegrees=5}
 {bone1Name=dyn_big_braid_r_01, bone2Name=dyn_big_braid_r_base, swing1LimitDegrees=30, swing2LimitDegrees=20, twistLimitDegrees=5}
 {
  bone1Name=dyn_small_braid_l_base
  bone2Name=head
  swing1LimitDegrees=45
  swing1Motion=Free
  swing2LimitDegrees=45
  swing2Motion=Free
  twistLimitDegrees=45
  twistMotion=Free
 }
 {bone1Name=dyn_small_braid_l_01, bone2Name=dyn_small_braid_l_base, swing1LimitDegrees=15, swing2LimitDegrees=10, twistLimitDegrees=5}
 {bone1Name=dyn_small_braid_l_02, bone2Name=dyn_small_braid_l_01, swing1LimitDegrees=12, swing2LimitDegrees=10, twistLimitDegrees=5}
 {bone1Name=dyn_small_braid_l_03, bone2Name=dyn_small_braid_l_02, swing1LimitDegrees=10, swing2LimitDegrees=10, twistLimitDegrees=5}
 {
  bone1Name=dyn_small_braid_r_base
  bone2Name=head
  swing1LimitDegrees=45
  swing1Motion=Free
  swing2LimitDegrees=45
  swing2Motion=Free
  twistLimitDegrees=45
  twistMotion=Free
 }
 {bone1Name=dyn_small_braid_r_01, bone2Name=dyn_small_braid_r_base, swing1LimitDegrees=15, swing2LimitDegrees=10, twistLimitDegrees=5}
 {bone1Name=dyn_small_braid_r_02, bone2Name=dyn_small_braid_r_01, swing1LimitDegrees=10, swing2LimitDegrees=10, twistLimitDegrees=5}
 {bone1Name=dyn_small_braid_r_03, bone2Name=dyn_small_braid_r_02, swing1LimitDegrees=10, swing2LimitDegrees=10, twistLimitDegrees=5}
 {
  bone1Name=dyn_quiff_l_base
  bone2Name=head
  swing1LimitDegrees=45
  swing1Motion=Free
  swing2LimitDegrees=45
  swing2Motion=Free
  twistLimitDegrees=45
  twistMotion=Free
 }
 {bone1Name=dyn_quiff_l_01, bone2Name=dyn_quiff_l_base, swing1LimitDegrees=40, swing2LimitDegrees=30, twistLimitDegrees=0}
 {
  bone1Name=dyn_quiff_r_base
  bone2Name=head
  swing1LimitDegrees=45
  swing1Motion=Free
  swing2LimitDegrees=45
  swing2Motion=Free
  twistLimitDegrees=45
  twistMotion=Free
 }
 {bone1Name=dyn_quiff_r_01, bone2Name=dyn_quiff_r_base, swing1LimitDegrees=40, swing2LimitDegrees=30, twistLimitDegrees=0}
 {
  bone1Name=spine_05
  bone2Name=spine_04
  swing1LimitDegrees=45
  swing1Motion=Free
  swing2LimitDegrees=45
  swing2Motion=Free
  twistLimitDegrees=45
  twistMotion=Free
 }
]
